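# Translucent parent material for the Aperture vehicle wheel cosmetic, built on MF_WheelParent_Simple with material attributes.
# Vertex color blue blends in an Opacity_Effect scalar on Opacity Mask and a rim-normal Fresnel times transformed reflection vector on emissive, scaled by Emissive_Effect.
# The modified attributes pass through vehicle damage, then de-rez transition functions.
# The Opacity_Mask texture branch is left disconnected from the result.
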Material /VehicleCosmetics/Wheels/Aperture/New/Materials/M_WheelParent_Aperture
graph {
 properties {
  bAutomaticallySetUsageInEditor=false
  bUseMaterialAttributes=true
  bUsedWithMeshDeformer=true
  bUsedWithSkeletalMesh=true
  blendMode=BLEND_Translucent
  pixelDepthOffsetMode=PDOM_Legacy
  previewMesh=/VehicleCosmetics/Wheels/Aperture/New/Meshes/SK_Wheel_Aperture
  usedShadingModels=Default Lit
 }
 nodes [
  Constant() {r=1}
  n1 = VertexColor()
  LinearInterpolate(A=Constant(), B=Constant() {r=1}, Alpha=@n1.B)
  n2 = TextureSampleParameter2D() {
   group=5_Special Effects
   parameterName=Opacity_Mask
   samplerType=SAMPLERTYPE_LinearColor
   texture=/VehicleCosmetics/Wheels/Aperture/New/Textures/T_Aperture_Rim_E
  }
  n3 = MaterialFunctionCall() {materialFunction=&MF_WheelParent_Simple}
  LinearInterpolate(A=Constant(), B=Constant() {r=1}, Alpha=LinearInterpolate(A=Constant(), B=@n2.R, Alpha=OneMinus(None=@n3.VertexColorR)))
  n4 = GetMaterialAttributes(None=@n3.MaterialAttributes) {attributeGetTypes [#1, #2]}
  n5 = LinearInterpolate(A=@n4.Opacity Mask, B=ScalarParameter() {defaultValue=1, group=5_Special Effects, parameterName=Opacity_Effect}, Alpha=@n1.B)
  n6 = Multiply(A=Transform(None=ReflectionVectorWS()), B=Constant() {r=2})
  n7 = MaterialFunctionCall!(
   Normal=TextureSample() {samplerType=SAMPLERTYPE_Normal, texture=/VehicleCosmetics/Wheels/Aperture/New/Textures/T_Aperture_Rim_N}.RGB
   Exponent=Constant() {r=1}
  ) {materialFunction=/VehicleCosmetics/Wheels/Cola_Holographic/Materials/MF_Fresnel_Classic}
  n8 = Multiply(
   A=LinearInterpolate(A=@n4.Emissive Color, B=Multiply(A=@n6, B=@n7.Result), Alpha=@n1.B)
   B=ScalarParameter() {defaultValue=1, group=5_Special Effects, parameterName=Emissive_Effect}
  )
  MaterialFunctionCall(In=MaterialFunctionCall(Material=SetMaterialAttributes!(MaterialAttributes=@n4.MaterialAttributes, Opacity Mask=@n5, Emissive Color=@n8) {attributeSetTypes [#1, #2]}) {materialFunction=&MF_BaseVehicle_Damage}.Result) {materialFunction=&MF_VehicleDeRez_Transition}
  Multiply(A=@n2.R)
  Multiply(A=@n3.VertexColorR)
  OneMinus(None=@n3.VertexColorR)
 ]
}
