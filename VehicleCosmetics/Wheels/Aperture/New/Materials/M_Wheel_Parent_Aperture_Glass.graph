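# Translucent, lit parent material for the glass on the Aperture vehicle cosmetic wheel, flagged for skeletal mesh use.
# A transformed world-space reflection vector serves directly as color, scaled by an Emissive parameter to give a view-dependent tint without textures.
# That result is multiplied by a classic Fresnel function borrowed from the Cola Holographic wheel.
# A separate scalar parameter provides uniform Opacity.
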Material /VehicleCosmetics/Wheels/Aperture/New/Materials/M_Wheel_Parent_Aperture_Glass
graph {
 properties {
  bAutomaticallySetUsageInEditor=false
  bUsedWithSkeletalMesh=true
  blendMode=BLEND_Translucent
  previewMesh=/VehicleCosmetics/Wheels/Aperture/New/Meshes/SK_Wheel_Aperture
  usedShadingModels=Default Lit
 }
 nodes [
  n1 = ReflectionVectorWS()
  Multiply(
   A=Multiply(A=Transform(None=@n1), B=ScalarParameter() {defaultValue=1, parameterName=Emissive})
   B=MaterialFunctionCall!(Normal, Exponent=Constant() {r=1}) {materialFunction=/VehicleCosmetics/Wheels/Cola_Holographic/Materials/MF_Fresnel_Classic}.Result
  )
  ScalarParameter() {defaultValue=0.5, parameterName=Opacity}
 ]
}
